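# Niagara system that renders pink-magenta impact bursts for PistolCanary tentacle damage, driven by Niagara Data Channel entries.
# Chunks, Sparks, SpikeyFlash (mesh plus sprite) and Streaks_Fat spawn through Spawn_NDCFortImpacts, taking position, surface normal and ImpactScale from the channel to orient and scale the burst.
# An enabled Debug emitter writes to NDC_PistolCanary_Tentacle_Damage using User.Scale.
# The Minimal debug writer and Spline_Streaks mesh emitter are disabled.
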NiagaraSystem /PistolCanary/Assets/FX/NS_NDC_PistolCanary_Tentacle_Damage
dependencies {
 usedDataInterfaces [
  &NiagaraDataInterfaceCurve
  &NiagaraDataInterfaceVector2DCurve
  /Script/Niagara.NiagaraDataInterfaceDataChannelWrite
  &NiagaraDataInterfaceVectorCurve
  /Script/Niagara.NiagaraDataInterfaceDataChannelRead
 ]
 usedDynamicInputs []
 usedModules [
  &SystemState
  &self:SystemSpawnScript.SystemScriptSource.SystemScriptGraph.NiagaraNodeAssignment_1.SetVariables_#1
  /Niagara/Modules/System/CompleteIfUnused
 ]
}
emitters {
 Chunks {
  emitterUpdateScript [
   &EmitterState()
   /Game/Effects/Niagara/DataChannels/Impacts/Spawn_NDCFortImpacts(Max Spawn Count=18, Min Spawn Count=16, Spawn Count Mode=2)
  ]
  particleSpawnScript [
   &self:Initialize_Particle_from_NDC() {module=Initialize_Particle_from_NDC}
   &KillParticles(Kill Particles=&SetBoolByFloatComparison(
     A=&RandomRangeFloat(Minimum=&Lerp_Float(A=0, B=0.5, Alpha=$Particles.ImpactScale), Maximum=1)
     B=0.75
     Comparison Type=::NewEnumerator6
    ))
   &InitializeParticle(
    Color Maximum=<rgba 2 0.1101 1.7034 1>
    Color Minimum=<rgba 0.25 0.0138 0.169 1>
    Color Mode=::NewEnumerator2
    Lifetime Max=1
    Lifetime Min=0.25
    Lifetime Mode=::NewEnumerator1
    Position=$Particles.Position
    Position Mode=::NewEnumerator1
    Sprite Size Mode=::NewEnumerator3
    Uniform Sprite Size=&Multiply_Float(
     A=&Lerp_Float(A=0.5, B=1, Alpha=$Particles.ImpactScale)
     B=&RandomRangeFloat(Minimum=30, Maximum=80)
    )
    UsePositionOffset=false
    Write Material Random=true
   )
   &ShapeLocation(MaskSpawnGroup=false, Override Seed=false, Sphere Radius=10, UseOwnerScale=true)
   &AddVelocity(
    Cone Angle=50
    Cone Axis=$Particles.ImpactSurfaceNormal
    Module.UseConeFalloff=false
    Module.UseVelDistribution=false
    Override Seed=false
    Velocity Mode=::NewEnumerator2
    Velocity Speed=&Lerp_Float(
     A=500
     B=5000
     Alpha=&FloatFromCurve(
      FloatCurve=DataInterface() {propertyValues {Curve {keys [
          {arriveTangent=0.5382, interpMode=RCIM_Cubic, leaveTangent=0.5382, tangentMode=RCTM_User, time=0, value=0}
          {arriveTangent=2.6808, interpMode=RCIM_Cubic, leaveTangent=2.6808, tangentMode=RCTM_User, time=1, value=1}
         ]}}}
      CurveIndex=&RandomRangeFloat(Minimum=0, Maximum=1)
      Scale Curve=&Lerp_Float(A=0.6, B=1, Alpha=$Particles.ImpactScale)
     )
    )
   )
  ]
  particleUpdateScript [
   &ParticleState()
   &GravityForce(Gravity=<0 0 -2200>)
   &Drag(
    Drag=&FloatFromCurve(
     FloatCurve=DataInterface() {propertyValues {Curve {keys [
         {interpMode=RCIM_Cubic, leaveTangent=-8.9049, leaveTangentWeight=0.5018, tangentMode=RCTM_Break, time=0, value=1}
         {interpMode=RCIM_Cubic, tangentMode=RCTM_Break, time=0.3, value=0.05}
        ]}}}
     Scale Curve=7
    )
    Use Linear Drag=true
    Use Rotational Drag=false
   )
   &SolveForcesAndVelocity(Clamp Velocity=false, Limit Acceleration=false)
   &ScaleSpriteSize(Uniform Curve Sprite Scale=DataInterface() {propertyValues {Curve {keys [
        {interpMode=RCIM_Cubic, tangentMode=RCTM_Break, time=0.4, value=1}
        {arriveTangent=-3.9054, interpMode=RCIM_Cubic, tangentMode=RCTM_Break, time=1, value=0}
       ]}}})
   &ScaleSpriteSizeBySpeed(Max Scale Factor=<0.7 10>, Velocity Threshold=2000) {enabled=false}
   &ScaleColor(
    Scale RGB=&VectorFromFloat(Value=&FloatFromCurve(
      FloatCurve=DataInterface() {propertyValues {Curve {keys [
          {arriveTangent=-1.136, interpMode=RCIM_Cubic, time=0, value=1}
          {arriveTangent=-2, interpMode=RCIM_Cubic, leaveTangent=-2, time=0.2, value=0.3}
          <at 0.5 0>
         ]}}}
      CurveIndex=&Multiply_Float(A=$Particles.Age, B=&RandomRangeFloat(Minimum=0.8, Maximum=1.2))
      Scale Curve=30
     ))
    ScaleA=false
    ScaleRGB=true
   ) {module=ScaleColor001}
   &DynamicMaterialParameters(
    Index 0 Param 1=&RandomRangeFloat(Minimum=2, Maximum=3)
    Index 0 Param 2=500
    Index 0 Param 3=1500
    Index 0 Param 4=5
    Index 1 Param 1=&RandomRangeFloat(Minimum=0, Maximum=1)
    Module.Index0Float0Write=true
    Module.Index0Float1Write=true
    Module.Index0Float2Write=true
    Module.Index0Float3Write=true
    Module.Index1Float0Write=true
    Module.Index1Float1Write=false
    Module.Index1Float2Write=false
    Module.Index1Float3Write=false
    Write Parameter Index 1=::NewEnumerator1
   )
  ]
  rendererClasses [&NiagaraSpriteRendererProperties]
  renderers [{rendererClass=&NiagaraSpriteRendererProperties}]
 }
 Debug {
  emitterUpdateScript [
   &EmitterState(Life Cycle Mode=::NewEnumerator1, Loop Duration=2, UseLoopDelay=false)
   &SpawnBurst_Instantaneous(Use Spawn Probability=false, UseLoopCountLimit=false)
  ]
  particleSpawnScript [
   &InitializeParticle(Color Mode=::NewEnumerator0, Lifetime=0.5, UsePositionOffset=false, Write Material Random=true)
   &self:ScratchModule(
    Data Channel Writer=DataInterface() {propertyValues {
      AllocationCount=1
      AllocationMode=Static
      Channel=&/NDC_PistolCanary_Tentacle_Damage
      bOnlyWriteOnceOnSubticks=true
      bPublishToCPU=true
      bPublishToGPU=true
      bPublishToGame=false
      bUpdateDestinationDataEveryTick=true
      AccessContext {accessContext {}}
     }}
    Scale:NiagaraFloat=$User.Scale
   ) {module=ScratchModule}
  ]
  particleUpdateScript [&ParticleState()]
  rendererClasses [&NiagaraSpriteRendererProperties]
  renderers [{rendererClass=&NiagaraSpriteRendererProperties}]
 }
 Minimal {
  bEnabled=false
  emitterUpdateScript [
   &EmitterState()
   &SpawnBurst_Instantaneous(Use Spawn Probability=false, UseLoopCountLimit=false)
  ]
  particleSpawnScript [
   &InitializeParticle(
    Color Mode=::NewEnumerator0
    Mesh Scale Mode=::NewEnumerator3
    Mesh Uniform Scale=15
    Sprite Size Mode=::NewEnumerator3
    Uniform Sprite Size=200
    UsePositionOffset=false
    Write Material Random=true
   )
   &self:Data_Channel_Debug(
    Data Channel Writer=DataInterface() {propertyValues {
      AllocationCount=1
      AllocationMode=Static
      Channel=&NDC_PistolCanary_Tentacle_Damage
      bOnlyWriteOnceOnSubticks=true
      bPublishToCPU=true
      bPublishToGPU=true
      bPublishToGame=true
      bUpdateDestinationDataEveryTick=true
      AccessContext {accessContext {}}
     }}
    ImpactNormal:Vector3f=$Engine.Owner.SystemXAxis
    ImpactPoint:NiagaraPosition=/Niagara/DynamicInputs/Helpers/SimulationPosition()
    ScaleMultiplier=1
   ) {module=Data_Channel_Debug}
  ]
  particleUpdateScript [&ParticleState()]
  rendererClasses [&NiagaraMeshRendererProperties]
  renderers [{rendererClass=&NiagaraMeshRendererProperties, rendererIndex=0}]
 }
 Sparks {
  emitterUpdateScript [
   &EmitterState()
   &Spawn_NDCFortImpacts(Max Spawn Count=20, Min Spawn Count=10, Spawn Count Mode=2)
  ]
  particleSpawnScript [
   &self:Initialize_Particle_from_NDC() {module=Initialize_Particle_from_NDC}
   &InitializeParticle(
    Color Maximum=<rgba 4 0 3.1393 1>
    Color Minimum=<rgba 1.02 0.04 2 1>
    Color Mode=::NewEnumerator2
    Lifetime=&FloatFromCurve(
     FloatCurve=DataInterface() {propertyValues {Curve {
        defaultValue=1
        keys [
         {arriveTangent=0.4191, interpMode=RCIM_Cubic, leaveTangent=0.4191, tangentMode=RCTM_User, time=0, value=0.1}
         {arriveTangent=2.3656, interpMode=RCIM_Cubic, leaveTangent=2.3656, tangentMode=RCTM_User, time=1, value=1}
        ]
       }}}
     CurveIndex=&RandomRangeFloat(Minimum=0, Maximum=1)
    )
    Position=$Particles.Position
    Position Mode=::NewEnumerator1
    Sprite Size Mode=::NewEnumerator1
    Uniform Sprite Size Max=7
    Uniform Sprite Size Min=4
    UsePositionOffset=false
    Write Material Random=true
   )
   &ShapeLocation(MaskSpawnGroup=false, Override Seed=false, Sphere Radius=10, UseOwnerScale=true)
   &AddVelocity(
    Cone Angle=110
    Cone Axis=$Particles.ImpactSurfaceNormal
    Inner Cone Angle=45
    Module.UseConeFalloff=true
    Module.UseVelDistribution=false
    Override Seed=false
    Speed Falloff From Cone Axis=0.2
    Velocity Mode=::NewEnumerator2
    Velocity Speed=&Lerp_Float(
     A=1500
     B=4500
     Alpha=&FloatFromCurve(
      FloatCurve=DataInterface() {propertyValues {Curve {keys [
          {arriveTangent=0.2237, interpMode=RCIM_Cubic, leaveTangent=0.2237, tangentMode=RCTM_User, time=0, value=0}
          {arriveTangent=2.6808, interpMode=RCIM_Cubic, leaveTangent=2.6808, tangentMode=RCTM_User, time=1, value=1}
         ]}}}
      CurveIndex=&RandomRangeFloat(Minimum=0, Maximum=1)
      Scale Curve=&Lerp_Float(A=0.5, B=1, Alpha=$Particles.ImpactScale)
     )
    )
   )
  ]
  particleUpdateScript [
   &ParticleState()
   &GravityForce(Gravity=<0 0 -1200>)
   &Drag(
    Drag=&FloatFromCurve(
     FloatCurve=DataInterface() {propertyValues {Curve {keys [
         {interpMode=RCIM_Cubic, leaveTangent=-8.9049, leaveTangentWeight=0.5018, tangentMode=RCTM_Break, time=0, value=1}
         {interpMode=RCIM_Cubic, tangentMode=RCTM_Break, time=0.3, value=0.025}
        ]}}}
     Scale Curve=&RandomRangeFloat(Minimum=5, Maximum=12)
    )
    Use Linear Drag=true
    Use Rotational Drag=false
   )
   &SolveForcesAndVelocity(Clamp Velocity=false, Limit Acceleration=false)
   &ScaleSpriteSize(Uniform Curve Sprite Scale=DataInterface() {propertyValues {Curve {keys [
        {interpMode=RCIM_Cubic, tangentMode=RCTM_Break, time=0.25, value=1}
        {arriveTangent=-2.298, interpMode=RCIM_Cubic, tangentMode=RCTM_Break, time=1, value=0.2}
       ]}}})
   &ScaleSpriteSizeBySpeed(
    Max Scale Factor=<1 10>
    Sample Scale Factor By Curve=true
    Scale Factor Curve=DataInterface() {propertyValues {Curve {keys [{arriveTangent=0.3851, interpMode=RCIM_Cubic, time=0, value=0}, {arriveTangent=2.2117, tangentMode=RCTM_User, time=1, value=1}]}}}
    Velocity Threshold=2000
   )
   &ScaleColor(
    Scale Alpha=&FloatFromCurve(FloatCurve=DataInterface() {propertyValues {Curve {keys [<at 0 1 RCIM_Cubic>, {arriveTangent=-2.9021, tangentMode=RCTM_User, time=1, value=0}]}}})
    ScaleA=true
    ScaleRGB=false
   ) {module=ScaleColor001}
   &DynamicMaterialParameters(
    Index 0 Param 1=&FloatFromCurve(
     FloatCurve=DataInterface() {propertyValues {Curve {defaultValue=1, keys [<at 0 0 RCIM_Cubic>, {arriveTangent=1.1412, tangentMode=RCTM_User, time=1, value=1}]}}}
     Scale Curve=2
    )
    Index 0 Param 3=3
    Module.Index0Float0Write=true
    Module.Index0Float1Write=false
    Module.Index0Float2Write=true
    Module.Index0Float3Write=false
   )
  ]
  rendererClasses [&NiagaraSpriteRendererProperties]
  renderers [{rendererClass=&NiagaraSpriteRendererProperties}]
 }
 SpikeyFlash {
  emitterUpdateScript [
   &EmitterState()
   &Spawn_NDCFortImpacts()
  ]
  particleSpawnScript [
   &self:Initialize_Particle_from_NDC() {module=Initialize_Particle_from_NDC}
   &InitializeParticle(
    Color=<rgba 2 0.0932 0.4976 1>
    Lifetime=&Lerp_Float(
     A=&RandomRangeFloat(Minimum=0.14, Maximum=0.17)
     B=&RandomRangeFloat(Minimum=0.18, Maximum=0.24)
     Alpha=$Particles.ImpactScale
    )
    Mesh Scale=&Multiply_VectorByFloat(
     Vector=&RandomRangeVector(Minimum=<0.1 0.1 0.5>, Maximum=<0.15 0.15 0.8>)
     Float=&Lerp_Float(A=0.5, B=1, Alpha=$Particles.ImpactScale)
    )
    Mesh Scale Mode=::NewEnumerator4
    Position=$Particles.Position
    Position Mode=::NewEnumerator1
    Sprite Rotation Mode=::NewEnumerator2
    Sprite Size Mode=::NewEnumerator3
    Uniform Sprite Size=&Multiply_Float(A=1500, B=&Lerp_Float(A=0.25, B=1, Alpha=$Particles.ImpactScale))
    UsePositionOffset=false
    Write Material Random=true
   )
   &CameraOffset(Camera Offset Amount=50)
   &InitialMeshOrientation(
    Mesh Axis To Orient=<0 0 1>
    Mesh Orientation Mode=::NewEnumerator2
    Orientation Vector=$Particles.ImpactSurfaceNormal
    Rotation=&RandomRangeVector(Minimum=<0 0 0>, Maximum=<0 0 1>)
    Use Rotation Vector=true
   )
   &SpriteFacingAndAlignment(Sprite Facing=$Particles.ImpactSurfaceNormal, WriteAlignment=false, WriteFacing=true)
  ]
  particleUpdateScript [
   &ParticleState()
   &ScaleSpriteSize(Uniform Curve Sprite Scale=DataInterface() {propertyValues {Curve {keys [
        {arriveTangent=6.3514, interpMode=RCIM_Cubic, leaveTangent=6.3514, tangentMode=RCTM_User, time=0, value=0}
        {tangentMode=RCTM_User, time=1, value=1}
       ]}}})
   &SolveForcesAndVelocity(Clamp Velocity=false, Limit Acceleration=false)
   &ScaleMeshSize(Scale Factor=&VectorFromFloat(Value=&FloatFromCurve(FloatCurve=DataInterface() {propertyValues {Curve {keys [
          {arriveTangent=9.901, interpMode=RCIM_Cubic, leaveTangent=9.901, tangentMode=RCTM_User, time=0, value=0.1}
          {arriveTangent=2.6083, interpMode=RCIM_Cubic, leaveTangent=2.6083, tangentMode=RCTM_User, time=1, value=6}
         ]}}})))
   &ScaleColor(
    Scale Alpha=&FloatFromCurve(FloatCurve=DataInterface() {propertyValues {Curve {keys [<at 0.4 1 RCIM_Cubic>, {arriveTangent=-2.2027, tangentMode=RCTM_User, time=1, value=0}]}}})
    Scale RGB=&VectorFromFloat(Value=2)
    ScaleA=true
    ScaleRGB=true
   )
   /SetVariables(Particles.Glow Sprite Color:LinearColor=&MakeLinearColorFromVectorAndFloat(
     "Vector (RGB)"=&MakeVectorFromLinearColorRGB(Color=<rgba 5 0.0138 0.5881 1>)
     "Float (Alpha)"=&FloatFromCurve(
      FloatCurve=DataInterface() {propertyValues {Curve {keys [
          {arriveTangent=-2.7319, interpMode=RCIM_Cubic, time=0.05, value=1}
          {arriveTangent=-1.5377, interpMode=RCIM_Cubic, leaveTangent=-1.5377, tangentMode=RCTM_User, time=0.2, value=0.55}
          <at 1 0>
         ]}}}
      Scale Curve=2
     )
    )) {isSetParameters=true}
  ]
  rendererClasses [&NiagaraMeshRendererProperties, &NiagaraSpriteRendererProperties]
  renderers [
   {rendererClass=&NiagaraMeshRendererProperties, rendererIndex=0}
   {rendererClass=&NiagaraSpriteRendererProperties, rendererIndex=1}
  ]
 }
 Spline_Streaks {
  bEnabled=false
  emitterUpdateScript [
   &EmitterState()
   &Spawn_NDCFortImpacts(Max Spawn Count=9, Min Spawn Count=7, Spawn Count Mode=2)
  ]
  particleSpawnScript [
   &self:Initialize_Particle_from_NDC() {module=Initialize_Particle_from_NDC}
   &KillParticles(Kill Particles=&SetBoolByFloatComparison(
     A=&RandomRangeFloat(Minimum=&Lerp_Float(A=0, B=0.5, Alpha=$Particles.ImpactScale), Maximum=1)
     B=0.75
     Comparison Type=::NewEnumerator6
    ))
   /SetVariables(Particles.RandomIntensity:NiagaraFloat=&FloatFromCurve(
     FloatCurve=DataInterface() {propertyValues {Curve {defaultValue=1, keys [<at 0 0 RCIM_Cubic>, {arriveTangent=1.6936, tangentMode=RCTM_User, time=1, value=1}]}}}
     CurveIndex=&RandomRangeFloat(Minimum=0, Maximum=&Lerp_Float(A=0.6, B=1, Alpha=$Particles.ImpactScale))
    )) {isSetParameters=true}
   &InitializeParticle(
    Color=<rgba 1.5371 0.0687 2 1>
    Lifetime=&Multiply_Float(
     A=&RandomRangeFloat(Minimum=0.5, Maximum=0.7)
     B=&Lerp_Float(A=0.5, B=1, Alpha=$Particles.RandomIntensity)
    )
    Mesh Scale Mode=::NewEnumerator3
    Mesh Uniform Scale=&Multiply_Float(
     A=&RandomRangeFloat(Minimum=0.7, Maximum=1)
     B=&Lerp_Float(A=1, B=6, Alpha=$Particles.RandomIntensity)
    )
    Position=$Particles.Position
    Position Mode=::NewEnumerator1
    UsePositionOffset=false
    Write Material Random=true
   )
   &ShapeLocation(
    Cylinder Height=5
    Cylinder Radius=10
    MaskSpawnGroup=false
    Override Seed=false
    Rotation Axis=<0 1 0>
    Rotation Mode=::NewEnumerator1
    Shape Primitive=::NewEnumerator1
    Surface Only=true
    Use Endcaps In Surface Only Mode=false
    UseOwnerScale=true
   ) {module=ShapeLocation001}
   &AddVelocity(
    Cone Angle=90
    Cone Axis=$Particles.ImpactSurfaceNormal
    Module.UseConeFalloff=false
    Module.UseVelDistribution=true
    Override Seed=false
    Velocity Mode=::NewEnumerator2
    Velocity Speed=&Lerp_Float(A=200, B=500, Alpha=$Particles.RandomIntensity)
   )
  ]
  particleUpdateScript [
   &ParticleState()
   &UpdateMeshOrientation(
    Rotation Rate=&FloatFromCurve(
     FloatCurve=DataInterface() {propertyValues {Curve {keys [
         {arriveTangent=-2.1925, interpMode=RCIM_Cubic, leaveTangent=-2.1925, tangentMode=RCTM_User, time=0, value=1}
         {arriveTangent=-0.2845, interpMode=RCIM_Cubic, leaveTangent=-0.2845, tangentMode=RCTM_User, time=1, value=0.05}
        ]}}}
     Scale Curve=&RandomRangeFloat(Minimum=0.15, Maximum=0.25)
    )
    Rotation Vector=<0 1 -1>
   )
   &ScaleMeshSize(Scale Factor=&VectorFromCurve(
     VectorCurve=DataInterface() {propertyValues {
       XCurve {keys [
         {arriveTangent=2.4453, interpMode=RCIM_Cubic, leaveTangent=2.4453, tangentMode=RCTM_User, time=0, value=0}
         {arriveTangent=0.1848, tangentMode=RCTM_User, time=1, value=1}
        ]}
       YCurve {keys [<at 0 1>,<at 1 0>]}
       ZCurve {keys [
         {arriveTangent=0.283, interpMode=RCIM_Cubic, leaveTangent=0.283, tangentMode=RCTM_User, time=0, value=0}
         {arriveTangent=2.0397, tangentMode=RCTM_User, time=1, value=1}
        ]}
      }}
     Scale Curve=<1 1 0.5>
    ))
   &GravityForce(Gravity=<0 0 -400>)
   &Drag(Drag=&RandomRangeFloat(Minimum=0.25, Maximum=0.5), Use Linear Drag=true, Use Rotational Drag=false)
   &SolveForcesAndVelocity(Clamp Velocity=false, Limit Acceleration=false)
   &DynamicMaterialParameters(
    Index 0 Param 1=&FloatFromCurve(FloatCurve=DataInterface() {propertyValues {Curve {keys [<at 0.4 0 RCIM_Cubic>, {arriveTangent=2.3043, interpMode=RCIM_Cubic, leaveTangent=2.3043, tangentMode=RCTM_User, time=1, value=1}]}}})
    Index 0 Param 2=&FloatFromCurve(
     FloatCurve=DataInterface() {propertyValues {Curve {keys [
         {arriveTangent=1.3993, interpMode=RCIM_Cubic, leaveTangent=1.3993, tangentMode=RCTM_User, time=0, value=0.5}
         {interpMode=RCIM_Cubic, tangentMode=RCTM_User, time=1, value=1}
        ]}}}
     Scale Curve=&RandomRangeFloat(Minimum=0, Maximum=2)
    )
    Index 0 Param 3=&Multiply_Float(
     A=&MakeFloatFromVector(Vector=$Particles.Initial.Scale, Channel=::NewEnumerator0)
     B=&RandomRangeFloat(Minimum=0.5, Maximum=2)
    )
    Index 0 Param 4=&FloatFromCurve(
     FloatCurve=DataInterface() {propertyValues {Curve {keys [
         {interpMode=RCIM_Cubic, leaveTangentWeight=0.7017, tangentMode=RCTM_User, time=0, value=0}
         {arriveTangent=1.7252, arriveTangentWeight=1.2352, interpMode=RCIM_Cubic, leaveTangent=1.7252, tangentMode=RCTM_User, time=1, value=1}
        ]}}}
     Scale Curve=&RandomRangeFloat(Minimum=0.75, Maximum=2)
    )
    Index 1 Param 1=&FloatFromCurve(
     FloatCurve=DataInterface() {propertyValues {Curve {keys [
         {arriveTangent=2.4162, interpMode=RCIM_Cubic, leaveTangent=2.4162, tangentMode=RCTM_User, time=0, value=0}
         {interpMode=RCIM_Cubic, tangentMode=RCTM_User, time=1, value=1}
        ]}}}
     Scale Curve=-1
    )
    Index 1 Param 2=&RandomRangeFloat(Minimum=0.75, Maximum=1.75)
    Module.Index0Float0Write=true
    Module.Index0Float1Write=true
    Module.Index0Float2Write=true
    Module.Index0Float3Write=true
    Module.Index1Float0Write=true
    Module.Index1Float1Write=true
    Module.Index1Float2Write=false
    Module.Index1Float3Write=false
    Write Parameter Index 1=::NewEnumerator1
   )
  ]
  rendererClasses [&NiagaraMeshRendererProperties]
  renderers [{rendererClass=&NiagaraMeshRendererProperties, rendererIndex=0}]
 }
 Streaks_Fat {
  emitterUpdateScript [
   &EmitterState()
   &Spawn_NDCFortImpacts(Max Spawn Count=5, Min Spawn Count=3, Spawn Count Mode=2)
  ]
  particleSpawnScript [
   &self:Initialize_Particle_from_NDC() {module=Initialize_Particle_from_NDC}
   &InitializeParticle(
    Color=<rgba 2 0.0932 0.4976 1>
    Lifetime Max=0.15
    Lifetime Min=0.11
    Lifetime Mode=::NewEnumerator1
    Position=$Particles.Position
    Position Mode=::NewEnumerator1
    Sprite Size=&Multiply_Vector2D_ByFloat(
     Float=&Lerp_Float(A=0.5, B=1, Alpha=$Particles.ImpactScale)
     Vector2D=&RandomRangeVector2D(Minimum=<12.5 275>, Maximum=<25 375>)
    )
    Sprite Size Mode=::NewEnumerator4
    UsePositionOffset=false
    Write Material Random=true
   )
   &ShapeLocation(MaskSpawnGroup=false, Override Seed=false, Sphere Radius=12, Sphere Surface Distribution=0.5595, UseOwnerScale=true)
   &AddVelocity(
    Cone Angle=10
    Cone Axis=$Particles.ImpactSurfaceNormal
    Inner Cone Angle=180
    Module.UseConeFalloff=false
    Module.UseVelDistribution=false
    Override Seed=false
    Velocity Mode=::NewEnumerator2
    Velocity Speed=&Lerp_Float(
     A=0
     B=1
     Alpha=&FloatFromCurve(
      FloatCurve=DataInterface() {propertyValues {Curve {keys [
          {interpMode=RCIM_Cubic, tangentMode=RCTM_User, time=0, value=0}
          {arriveTangent=2.6808, interpMode=RCIM_Cubic, leaveTangent=2.6808, tangentMode=RCTM_User, time=1, value=1}
         ]}}}
      CurveIndex=&RandomRangeFloat(Minimum=0, Maximum=1)
     )
    )
   )
  ]
  particleUpdateScript [
   &ParticleState()
   &SolveForcesAndVelocity(Clamp Velocity=false, Limit Acceleration=false)
   &ScaleSpriteSize(Scale Factor=&Vector2DFromCurve(
     Vector2Curve=DataInterface() {propertyValues {XCurve {keys [<at 0 0.75 RCIM_Cubic>,<at 1 1 RCIM_Cubic>]}, YCurve {keys [<at 0 0 RCIM_Cubic>,<at 1 2 RCIM_Cubic>]}}}
     ScaleCurve=<1 1>
    ))
   &ScaleColor(
    Scale Alpha=&FloatFromCurve(FloatCurve=DataInterface() {propertyValues {Curve {keys [<at 0 1 RCIM_Cubic>,<at 1 0>]}}})
    Scale RGB=&VectorFromFloat(Value=&FloatFromCurve(FloatCurve=DataInterface() {propertyValues {Curve {keys [<at 0 10 RCIM_Cubic>,<at 1 1 RCIM_Cubic>]}}}))
    ScaleA=true
    ScaleRGB=true
   )
  ]
  rendererClasses [&NiagaraSpriteRendererProperties]
  renderers [{rendererClass=&NiagaraSpriteRendererProperties}]
 }
}
summary {
 systemName=NS_NDC_PistolCanary_Tentacle_Damage
 emitters [
  {emitterName=Sparks, rendererClasses [&NiagaraSpriteRendererProperties]}
  {emitterName=Streaks_Fat, rendererClasses [&NiagaraSpriteRendererProperties]}
  {bEnabled=false, emitterName=Minimal, rendererClasses [&NiagaraMeshRendererProperties]}
  {emitterName=SpikeyFlash, rendererClasses [&NiagaraMeshRendererProperties, &NiagaraSpriteRendererProperties]}
  {emitterName=Debug, rendererClasses [&NiagaraSpriteRendererProperties]}
  {bEnabled=false, emitterName=Spline_Streaks, rendererClasses [&NiagaraMeshRendererProperties]}
  {emitterName=Chunks, rendererClasses [&NiagaraSpriteRendererProperties]}
 ]
 userVariables [User.Scale {type=NiagaraFloat, defaultValue=0, description=""}]
}
