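# Niagara scorch effect for PistolCanary's tentacle, combining hot debris chunks, GPU embers, flickering glows, smoke and fading burn decals.
# Glows and decals store their offset from a skeletal mesh bone at spawn, then follow it to stay attached.
# Smoke spawns at the system and also from Decal particles through an attribute reader.
# User.Scale drives sizes, rates and spread, while User.SpawnRateMultiply scales glow, decal and smoke spawning.
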NiagaraSystem /PistolCanary/Assets/FX/NS_PistolCanary_Tentacle_Scorch
dependencies {
 usedDataInterfaces [
  &NiagaraDataInterfaceCurve
  &NiagaraDataInterfaceColorCurve
  &NiagaraDataInterfaceParticleRead
  &NiagaraDataInterfaceSkeletalMesh
 ]
 usedDynamicInputs []
 usedModules [
  &self:SystemSpawnScript.SystemScriptSource.SystemScriptGraph.NiagaraNodeAssignment_0.SetVariables_#1
  &SystemState
 ]
}
emitters {
 Chunks {
  emitterUpdateScript [
   &EmitterState()
   &SpawnRate(SpawnRate=&Lerp_Float(A=50, B=100, Alpha=$User.Scale), Use Spawn Probability=false)
  ]
  particleSpawnScript [
   &InitializeParticle(
    Color Maximum=<rgba 5 0.5757 0 1>
    Color Minimum=<rgba 2 0 0.0045 1>
    Color Mode=::NewEnumerator2
    Lifetime Max=3
    Lifetime Min=0.5
    Lifetime Mode=::NewEnumerator1
    Sprite Size Mode=::NewEnumerator3
    Uniform Sprite Size=&RandomRangeFloat(Minimum=30, Maximum=80)
    UsePositionOffset=false
    Write Material Random=true
   )
   &ShapeLocation(MaskSpawnGroup=false, Override Seed=false, Sphere Radius=40, UseOwnerScale=true)
   &AddVelocity(
    Cone Angle=50
    Module.UseConeFalloff=false
    Module.UseVelDistribution=false
    Override Seed=false
    Velocity Mode=::NewEnumerator2
    Velocity Speed=&Lerp_Float(
     A=1000
     B=6000
     Alpha=&FloatFromCurve(
      FloatCurve=DataInterface() {propertyValues {Curve {keys [
          {arriveTangent=0.5382, interpMode=RCIM_Cubic, leaveTangent=0.5382, tangentMode=RCTM_User, time=0, value=0}
          {arriveTangent=2.6808, interpMode=RCIM_Cubic, leaveTangent=2.6808, tangentMode=RCTM_User, time=1, value=1}
         ]}}}
      CurveIndex=&RandomRangeFloat(Minimum=0, Maximum=1)
     )
    )
   )
  ]
  particleUpdateScript [
   &ParticleState()
   &GravityForce(Gravity=<0 0 -2200>)
   &Drag(
    Drag=&FloatFromCurve(
     FloatCurve=DataInterface() {propertyValues {Curve {keys [
         {interpMode=RCIM_Cubic, leaveTangent=-8.9049, leaveTangentWeight=0.5018, tangentMode=RCTM_Break, time=0, value=1}
         {interpMode=RCIM_Cubic, tangentMode=RCTM_Break, time=0.3, value=0.05}
        ]}}}
     Scale Curve=5
    )
    Use Linear Drag=true
    Use Rotational Drag=false
   )
   &SolveForcesAndVelocity(Clamp Velocity=false, Limit Acceleration=false)
   &ScaleSpriteSize(Uniform Curve Sprite Scale=DataInterface() {propertyValues {Curve {keys [
        {interpMode=RCIM_Cubic, tangentMode=RCTM_Break, time=0.4, value=1}
        {arriveTangent=-3.9054, interpMode=RCIM_Cubic, tangentMode=RCTM_Break, time=1, value=0}
       ]}}})
   &ScaleColor(
    Scale RGB=&VectorFromFloat(Value=&FloatFromCurve(
      FloatCurve=DataInterface() {propertyValues {Curve {keys [
          {arriveTangent=-1.136, interpMode=RCIM_Cubic, time=0, value=1}
          {arriveTangent=-1.3333, interpMode=RCIM_Cubic, leaveTangent=-1.3333, time=0.25, value=0.3}
          <at 0.75 0>
         ]}}}
      CurveIndex=&Multiply_Float(A=$Particles.Age, B=&RandomRangeFloat(Minimum=0.8, Maximum=1.2))
      Scale Curve=30
     ))
    ScaleA=false
    ScaleRGB=true
   ) {module=ScaleColor001}
   /SetVariables(
    Particles.GlowColor=&MultiplyLinearColor(Linear Color=$Particles.Color, Scale Factor=<1 1 1 0.03>)
    Particles.GlowSize=&Multiply_Vector2D(A=$Particles.SpriteSize, B=<2.5 4>)
   ) {isSetParameters=true}
   &DynamicMaterialParameters(
    Index 0 Param 1=&RandomRangeFloat(Minimum=2, Maximum=3)
    Index 0 Param 2=500
    Index 0 Param 3=1500
    Index 0 Param 4=6
    Index 1 Param 1=&RandomRangeFloat(Minimum=0, Maximum=1)
    Index 3 Param 1=3
    Index 3 Param 2=1
    Module.Index0Float0Write=true
    Module.Index0Float1Write=true
    Module.Index0Float2Write=true
    Module.Index0Float3Write=true
    Module.Index1Float0Write=true
    Module.Index1Float1Write=false
    Module.Index1Float2Write=false
    Module.Index1Float3Write=false
    Module.Index3Float0Write=true
    Module.Index3Float1Write=true
    Module.Index3Float2Write=false
    Module.Index3Float3Write=false
    Write Parameter Index 1=::NewEnumerator1
    Write Parameter Index 3=::NewEnumerator1
   )
  ]
  rendererClasses [&NiagaraSpriteRendererProperties]
  renderers [
   {rendererClass=&NiagaraSpriteRendererProperties}
   {rendererClass=&NiagaraSpriteRendererProperties, rendererIndex=1}
  ]
 }
 Decal {
  emitterUpdateScript [
   &EmitterState()
   &SpawnRate(SpawnRate=&Multiply_Float(A=$User.SpawnRateMultiply, B=4), Use Spawn Probability=false)
  ]
  particleSpawnScript [
   &InitializeParticle(
    Color=<rgba 1 0.1092 0.0039 0.1>
    Lifetime=7
    Mesh Scale Mode=::NewEnumerator3
    Mesh Uniform Scale=50
    Sprite Rotation Mode=::NewEnumerator2
    Sprite Size Mode=::NewEnumerator3
    Uniform Sprite Size=1000
    UsePositionOffset=false
    Write Material Random=true
   )
   &CameraOffset(Camera Offset Amount=500)
   /SetVariables(Particles.DecalSize:Vector3f=&Lerp_Vector(A=<1000 900 900>, B=<1500 3000 3000>, Alpha=$User.Scale)) {isSetParameters=true}
   &SampleSkeletalMesh(
    AdditionalYawPitchRoll=false
    Bone / Socket Index=3
    Bone Sampling Mode:ENiagaraBoneSamplingMode=::NewEnumerator4
    Skeletal Mesh:NiagaraDataInterfaceSkeletalMesh=$System.Skeletal Mesh
   )
   /SetVariables(Particles.InitialWorldPosition:NiagaraPosition=&Subtract_Position(A=$Engine.Owner.Position, B=$Particles.SampleSkeletalMesh.SampledPosition)) {isSetParameters=true}
   /SetVariables(Particles.InitialWorldFacing:Vector3f=&Multiply_VectorByFloat(Vector=$Engine.Owner.SystemXAxis, Float=-1)) {isSetParameters=true}
   &InitialMeshOrientation(
    Mesh Orientation Mode=::NewEnumerator2
    Rotation=&RandomRangeVector(Minimum=<0 0 0>, Maximum=<1 0 0>)
    Use Rotation Vector=true
   )
  ]
  particleUpdateScript [
   &ParticleState()
   &SkeletalMeshLocation(
    Bone / Socket Index=3
    Bone Sampling Mode=::NewEnumerator4
    MaskSpawnGroup=false
    Orientation Sampling=::NewEnumerator1
    Sampled Position Offset:NiagaraPosition=$Particles.InitialWorldPosition
    Sampled Position Offset Coordinate Space:ENiagaraSkelMeshTransforms=::NewEnumerator1
    Skeletal Mesh=$System.Skeletal Mesh
   )
   &UpdateMeshOrientation(
    Facing Coordinate Space:ENiagaraCoordinateSpace=::Local
    Facing Direction=&TransformVector(Vector=$Particles.InitialWorldFacing, Source Space=::World, Destination Space=::Local)
    Facing Interpolation=false
    Orientation Method=::NewEnumerator1
    Override Quat=false
    TransformFacing=true
    TransformUp=false
   )
   /SetVariables(Particles.DecalColor:LinearColor=&MakeLinearColorFromVectorAndFloat(
     "Vector (RGB)"=&Lerp_Vector(
      A=<1 0.2 0.2>
      B=<5 5 5>
      Alpha=&FloatFromCurve(
       FloatCurve=DataInterface() {propertyValues {Curve {keys [
           {arriveTangent=44.4092, interpMode=RCIM_Cubic, time=0, value=0.1}
           {arriveTangent=0.25, interpMode=RCIM_Cubic, leaveTangent=0.25, time=0.16, value=1}
           {arriveTangent=-0.7774, interpMode=RCIM_Cubic, leaveTangent=-0.7774, tangentMode=RCTM_User, time=0.4, value=0.2}
           <at 1 0>
          ]}}}
       CurveIndex=$Particles.Age
      )
     )
     "Float (Alpha)"=&FloatFromCurve(
      FloatCurve=DataInterface() {propertyValues {Curve {keys [
          {arriveTangent=5.9491, interpMode=RCIM_Cubic, time=0, value=0}
          <at 0.1 1>
          {arriveTangent=-1.1111, interpMode=RCIM_Cubic, leaveTangent=-1.1111, time=0.9, value=1}
          <at 1 0 RCIM_Cubic>
         ]}}}
      Scale Curve=2
     )
    )) {isSetParameters=true}
  ]
  rendererClasses [&NiagaraDecalRendererProperties, &NiagaraMeshRendererProperties]
  renderers [
   {rendererClass=&NiagaraDecalRendererProperties, rendererIndex=0}
   {rendererClass=&NiagaraMeshRendererProperties, rendererIndex=1}
  ]
 }
 EmbersHigh {
  simTarget=GPUComputeSim
  emitterUpdateScript [
   &EmitterState()
   &SpawnRate(SpawnRate=&Lerp_Float(A=90, B=150, Alpha=$User.Scale), Use Spawn Probability=false)
  ]
  particleSpawnScript [
   /SetVariables(Particles.SparkIntensity:NiagaraFloat=&FloatFromCurve(
     FloatCurve=DataInterface() {propertyValues {Curve {keys [
         {arriveTangent=0.2996, leaveTangentWeight=0.2, time=0, value=0}
         {
          arriveTangent=0.7746
          arriveTangentWeight=0.5371
          interpMode=RCIM_Cubic
          leaveTangent=0.7746
          leaveTangentWeight=0.3581
          tangentMode=RCTM_User
          time=0.5
          value=0.4
         }
         {arriveTangentWeight=0.1333, interpMode=RCIM_Cubic, tangentMode=RCTM_User, time=1, value=1}
        ]}}}
     CurveIndex=&RandomRangeFloat(Minimum=0, Maximum=1)
    )) {isSetParameters=true}
   &InitializeParticle(
    Color=&LerpLinearColors(StartColor=<rgba 8 0.6838 0.2252 1>, EndColor=<rgba 10 5.7312 0.265 1>, LerpFactor=$Particles.SparkIntensity)
    Lifetime=&RandomRangeFloat(Minimum=0.5, Maximum=2)
    Mass=&Lerp_Float(A=0.5, B=1, Alpha=$Particles.SparkIntensity)
    Mass Mode=::NewEnumerator1
    Sprite Size Mode=::NewEnumerator3
    Uniform Sprite Size=&Lerp_Float(A=20, B=27, Alpha=$Particles.SparkIntensity)
    UsePositionOffset=false
    Write Material Random=true
   )
   &ShapeLocation(
    Cylinder Height=300
    Cylinder Radius=&Lerp_Float(A=300, B=500, Alpha=$User.Scale)
    MaskSpawnGroup=false
    Override Seed=false
    Shape Primitive=::NewEnumerator1
    Surface Only=false
    UseOwnerScale=true
   )
   &AddVelocity(
    Module.Use Falloff Distance=false
    Override Seed=false
    Velocity Mode=::NewEnumerator1
    Velocity Speed=&Multiply_Float(
     A=&Multiply_Float(
      A=&RandomRangeFloat(Minimum=500, Maximum=2500)
      B=&Lerp_Float(A=0.4, B=1, Alpha=$Particles.SparkIntensity)
     )
     B=&MaskFloatBySpawnGroup(Masked Float=0.1, Passthrough Float=1, Spawn Group Mask=1)
    )
   )
   &CameraOffset(Camera Offset Amount=300)
  ]
  particleUpdateScript [
   &ParticleState()
   &WindForce(
    Frequency=3
    Friction Mode:ENiagaraWindFrictionMode=::NewEnumerator0
    Scale=5
    Surface Alignment Mode:ENiagaraWindCombingMode=::NewEnumerator0
    Wind Speed=$Engine.Owner.SystemXAxis
    Wind Speed Scale=1200
   )
   &Drag(Drag=&RandomRangeFloat(Minimum=1, Maximum=2), Use Linear Drag=true, Use Rotational Drag=false)
   &DynamicMaterialParameters(
    Index 0 Param 3=&FloatFromCurve(FloatCurve=DataInterface() {propertyValues {Curve {
        defaultValue=3
        keys [
         {interpMode=RCIM_Cubic, tangentMode=RCTM_User, time=0, value=0}
         {interpMode=RCIM_Cubic, tangentMode=RCTM_User, time=0.5, value=1}
         {tangentMode=RCTM_User, time=1, value=0}
        ]
       }}})
    Module.Index0Float0Write=false
    Module.Index0Float1Write=false
    Module.Index0Float2Write=true
    Module.Index0Float3Write=false
   )
   &SolveForcesAndVelocity(Clamp Velocity=false, Limit Acceleration=false)
   &ScaleColor(
    Linear Color Curve=DataInterface() {propertyValues {
      AlphaCurve {keys [<at 0 1>,<at 1 1>]}
      BlueCurve {keys [{arriveTangent=-2.9361, interpMode=RCIM_Cubic, leaveTangent=-2.9361, tangentMode=RCTM_User, time=0, value=1}, <at 1 0.151>]}
      GreenCurve {keys [{arriveTangent=-1.4408, interpMode=RCIM_Cubic, leaveTangent=-1.4408, tangentMode=RCTM_User, time=0.1467, value=1}, <at 1 0.1631>]}
      RedCurve {keys [{arriveTangent=-0.8332, interpMode=RCIM_Cubic, leaveTangent=-0.8332, tangentMode=RCTM_User, time=0.1467, value=1}, <at 1 1>]}
     }}
    Scale Mode=::NewEnumerator2
   )
   &ScaleSpriteSizeBySpeed(
    Max Scale Factor=<1 10>
    Sample Scale Factor By Curve=true
    Scale Factor Curve=DataInterface() {propertyValues {Curve {keys [
        {arriveTangent=0.2757, time=0, value=0}
        {arriveTangent=0.3844, interpMode=RCIM_Cubic, leaveTangent=0.3844, tangentMode=RCTM_User, time=0.4, value=0.15}
        {arriveTangent=2.2765, interpMode=RCIM_Cubic, leaveTangent=2.2765, tangentMode=RCTM_User, time=1, value=1}
       ]}}}
    Velocity Threshold=5000
   )
   &ScaleSpriteSize(
    Scale Factor=&Vector2DFromFloat(Value=&FloatFromCurve(FloatCurve=DataInterface() {propertyValues {Curve {keys [<at 0.5 1 RCIM_Cubic>, {arriveTangent=-3.8662, interpMode=RCIM_Cubic, leaveTangent=-3.8662, tangentMode=RCTM_User, time=1, value=0}]}}}))
    Scale Sprite Size Mode=::NewEnumerator2
   )
  ]
  rendererClasses [&NiagaraSpriteRendererProperties]
  renderers [{rendererClass=&NiagaraSpriteRendererProperties}]
 }
 Glows {
  emitterUpdateScript [
   &EmitterState()
   &SpawnRate(SpawnRate=&Multiply_Float(A=15, B=$User.SpawnRateMultiply), Use Spawn Probability=false)
  ]
  particleSpawnScript [
   &InitializeParticle(
    Position Offset=<10 0 0>
    Position Offset Coordinate Space=::Local
    Sprite Size Mode=::NewEnumerator3
    Uniform Sprite Size=&Multiply_Float(
     A=&RandomRangeFloat(Minimum=400, Maximum=700)
     B=&Lerp_Float(A=0.5, B=1, Alpha=$User.Scale)
    )
    UsePositionOffset=true
    Write Material Random=true
   )
   &SampleSkeletalMesh(
    AdditionalYawPitchRoll=false
    Bone / Socket Index=3
    Bone Sampling Mode:ENiagaraBoneSamplingMode=::NewEnumerator4
    Skeletal Mesh:NiagaraDataInterfaceSkeletalMesh=$System.Skeletal Mesh
   )
   /SetVariables(Particles.InitialWorldFacing:Vector3f=$Engine.Owner.SystemXAxis) {isSetParameters=true}
   /SetVariables(Particles.InitialWorldPosition:NiagaraPosition=&Subtract_Position(A=$Engine.Owner.Position, B=$Particles.SampleSkeletalMesh.SampledPosition)) {isSetParameters=true}
   &CameraOffset(Camera Offset Amount=200)
   &DynamicMaterialParameters(
    Index 0 Param 1=3
    Index 0 Param 2=1
    Module.Index0Float0Write=true
    Module.Index0Float1Write=true
    Module.Index0Float2Write=false
    Module.Index0Float3Write=false
   )
  ]
  particleUpdateScript [
   &ParticleState()
   &SkeletalMeshLocation(
    Bone / Socket Index=3
    Bone Sampling Mode=::NewEnumerator4
    MaskSpawnGroup=false
    Sampled Position Offset:NiagaraPosition=$Particles.InitialWorldPosition
    Sampled Position Offset Coordinate Space:ENiagaraSkelMeshTransforms=::NewEnumerator1
    Skeletal Mesh=$System.Skeletal Mesh
   )
   &SpriteFacingAndAlignment(
    Facing Coordinate Space=::Local
    Sprite Facing=&TransformVector(Vector=$Particles.InitialWorldFacing, Source Space=::World, Destination Space=::Local)
    WriteAlignment=false
    WriteFacing=true
   )
   /SetVariables(Particles.Flicker:NiagaraFloat=&Waveform(
     "Waveform [1]"=::NewEnumerator12
     "[1] Frequency"=&RandomRangeFloat(Minimum=16, Maximum=18)
     "[1] Amplitude Min/Max"=<0.9 1.1>
    )) {isSetParameters=true}
   &ScaleSpriteSize(
    Scale Sprite Size Mode=::NewEnumerator0
    Uniform Scale Factor=&FloatFromCurve(
     FloatCurve=DataInterface() {propertyValues {Curve {keys [
         {arriveTangent=31.2071, interpMode=RCIM_Cubic, leaveTangent=31.2071, tangentMode=RCTM_User, time=0, value=0}
         {arriveTangent=-2.2005, interpMode=RCIM_Cubic, leaveTangent=-2.2005, tangentMode=RCTM_User, time=0.1, value=1.6}
         <at 1 1>
        ]}}}
     Scale Curve=$Particles.Flicker
    )
   )
   &ScaleColor(
    Scale Alpha=&FloatFromCurve(FloatCurve=DataInterface() {propertyValues {Curve {
        defaultValue=1
        keys [
         {arriveTangent=11.9592, interpMode=RCIM_Cubic, leaveTangent=11.9592, tangentMode=RCTM_User, time=0, value=0}
         {arriveTangentWeight=0.0648, interpMode=RCIM_Cubic, leaveTangentWeight=0.3239, tangentMode=RCTM_User, time=0.1, value=1}
         {arriveTangentWeight=0.0997, interpMode=RCIM_Cubic, leaveTangentWeight=0.3488, tangentMode=RCTM_User, time=0.6, value=1}
         {arriveTangent=-2.0005, tangentMode=RCTM_User, time=1, value=0}
        ]
       }}})
    Scale RGB=&MakeVectorFromLinearColorRGB(Color=&ColorFromCurve(ColorCurve=DataInterface() {propertyValues {
        AlphaCurve {keys [<at 0 1 RCIM_Cubic>]}
        BlueCurve {keys [<at 0 0.1459 RCIM_Cubic>,<at 0.2678 0.0196>,<at 0.7656 0.0293 RCIM_Cubic>]}
        GreenCurve {keys [<at 0 1.1077 RCIM_Cubic>,<at 0.2678 0.0578>,<at 0.7656 0.0484 RCIM_Cubic>]}
        RedCurve {keys [<at 0 5 RCIM_Cubic>,<at 0.2678 2>,<at 0.7656 1 RCIM_Cubic>]}
       }}))
    ScaleA=true
    ScaleRGB=true
   )
  ]
  rendererClasses [&NiagaraSpriteRendererProperties]
  renderers [{rendererClass=&NiagaraSpriteRendererProperties}]
 }
 Smoke {
  emitterUpdateScript [
   &EmitterState()
   &SpawnRate(
    SpawnRate=&Multiply_Float(A=&Lerp_Float(A=10, B=20, Alpha=$User.Scale), B=$User.SpawnRateMultiply)
    Use Spawn Probability=false
   )
   &SpawnParticlesFromOtherEmitter(
    Attribute Reader=DataInterface() {propertyValues {EmitterBinding {bindingMode=Other, emitterName=Decal}}}
    Spawn Group=1
    Spawn Rate=5
    Use Maximum Age=false
    Use Minimum Age=false
   )
  ]
  particleSpawnScript [
   &InitializeParticle(
    Color=<rgba 0.02 0.0174 0.0188 0.95>
    Lifetime=&RandomRangeFloat(Minimum=1, Maximum=2)
    Sprite Rotation Mode=::NewEnumerator2
    Sprite Size Mode=::NewEnumerator3
    Uniform Sprite Size=&Multiply_Float(
     A=&RandomRangeFloat(Minimum=2500, Maximum=3500)
     B=&Lerp_Float(A=0.6, B=1, Alpha=$User.Scale)
    )
    UsePositionOffset=false
    Write Material Random=true
   )
   &SampleParticlesFromOtherEmitter()
   &ShapeLocation(
    Cylinder Height Midpoint=1
    Cylinder Radius=&Lerp_Float(A=300, B=500, Alpha=$User.Scale)
    MaskSpawnGroup=false
    Override Seed=false
    Shape Primitive=::NewEnumerator1
    Surface Only=false
    UseOwnerScale=true
   )
   /SetVariables(Particles.InitSpriteRotRate=&UniformRangedFloat(Minimum=-3, Maximum=3)) {isSetParameters=true}
   /SetVariables(Particles.RandomFloat=&RandomRangeFloat(Minimum=0, Maximum=1)) {isSetParameters=true}
   /SetVariables(Particles.SubUVRandomStart:NiagaraInt32=&RandomRangeInt(Minimum=0, Maximum=44)) {isSetParameters=true}
   &AddVelocity(
    Cone Angle=60
    Module.UseConeFalloff=false
    Module.UseVelDistribution=false
    Override Seed=false
    Velocity Mode=::NewEnumerator2
    Velocity Speed=&MaskFloatBySpawnGroup(
     Masked Float=&RandomRangeFloat(Minimum=500, Maximum=2000)
     Passthrough Float=&Multiply_Float(
      A=&RandomRangeFloat(Minimum=1500, Maximum=5000)
      B=&Lerp_Float(A=0.5, B=1, Alpha=$User.Scale)
     )
     Spawn Group Mask=1
    )
   )
  ]
  particleUpdateScript [
   &ParticleState()
   &Drag(Drag=2, Use Linear Drag=true, Use Rotational Drag=false)
   &SolveForcesAndVelocity(Clamp Velocity=false, Limit Acceleration=false)
   &ScaleColor(
    Scale Alpha=&FloatFromCurve(FloatCurve=DataInterface() {propertyValues {Curve {keys [
         {arriveTangent=43.8033, interpMode=RCIM_Cubic, time=0, value=0}
         <at 0.25 1 RCIM_Cubic>
         {arriveTangent=-2.8482, interpMode=RCIM_Cubic, leaveTangent=-2.8482, tangentMode=RCTM_User, time=1, value=0}
        ]}}})
    ScaleA=true
    ScaleRGB=false
   )
   &SpriteRotationRate(Rotation Rate=&Multiply_Float(
     A=$Particles.InitSpriteRotRate
     B=&FloatFromCurve(
      FloatCurve=DataInterface() {propertyValues {Curve {keys [
          {arriveTangent=-2.157, interpMode=RCIM_Cubic, leaveTangent=-2.157, tangentMode=RCTM_User, time=-0.0046, value=1.0113}
          {interpMode=RCIM_Cubic, leaveTangent=-0.2496, time=1, value=0.16}
         ]}}}
      Scale Curve=20
     )
    ))
   &ScaleSpriteSize(Uniform Curve Sprite Scale=DataInterface() {propertyValues {Curve {keys [
        {arriveTangent=1.1098, interpMode=RCIM_Cubic, leaveTangent=1.1098, tangentMode=RCTM_User, time=0, value=0.2}
        {arriveTangent=0.3088, tangentMode=RCTM_User, time=1, value=1}
       ]}}})
   &CameraOffset(Camera Offset Amount=/Niagara/DynamicInputs/Math/CalculateParticleRadius(
     Radius Calculation Type:ENiagaraCollisionRadiusOptions=::NewEnumerator0
     Particle Radius:NiagaraFloat=1
     Mesh Dimensions:Vector3f=<1 1 1>
     Method for Calculating Particle Radius:ENiagaraCalculateRadiusOptions=::NewEnumerator2
    ))
   /SetVariables(Particles.SubImageIndex=&Lerp_Float(A=&RandomRangeFloat(Minimum=4, Maximum=12), B=63, Alpha=$Particles.NormalizedAge)) {isSetParameters=true}
  ]
  rendererClasses [&NiagaraSpriteRendererProperties]
  renderers [{rendererClass=&NiagaraSpriteRendererProperties}]
 }
}
summary {
 systemName=NS_PistolCanary_Tentacle_Scorch
 emitters [
  {emitterName=Chunks, rendererClasses [&NiagaraSpriteRendererProperties]}
  {emitterName=Glows, rendererClasses [&NiagaraSpriteRendererProperties]}
  {emitterName=Decal, rendererClasses [&NiagaraDecalRendererProperties, &NiagaraMeshRendererProperties]}
  {emitterName=Smoke, rendererClasses [&NiagaraSpriteRendererProperties]}
  {emitterName=EmbersHigh, simTarget=GPUComputeSim, rendererClasses [&NiagaraSpriteRendererProperties]}
 ]
 userVariables [
  User.Scale {type=NiagaraFloat, defaultValue=1, description=""}
  User.SpawnRateMultiply {type=NiagaraFloat, defaultValue=1, description=""}
 ]
}
