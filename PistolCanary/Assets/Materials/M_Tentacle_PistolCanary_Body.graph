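# Skeletal-mesh body material for the PistolCanary tentacle, using material attributes, per-expression shading models and a slimy-sack subsurface profile.
# Between shared character PreFX and PostFX functions, it layers slimy sack, leathery skin and pulsing veins using Tech Art mask channels.
# The leathery skin alpha combines its own mask with the vein mask.
# A static switch then applies either goopy interior, driven by mask alpha, or world glow before PostFX.
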Material /PistolCanary/Assets/Materials/M_Tentacle_PistolCanary_Body
graph {
 properties {
  bAutomaticallySetUsageInEditor=false
  bUseMaterialAttributes=true
  bUsedWithMeshDeformer=true
  bUsedWithSkeletalMesh=true
  shadingModel=MSM_FromMaterialExpression
  subsurfaceProfile=&/SSP_SlimySack
  usedShadingModels=Default Lit | Subsurface | Clear Coat | Subsurface Profile | Cloth
 }
 nodes [
  n1 = MaterialFunctionCall() {materialFunction=&MF_CharacterParent_2023_PreFX}
  ChannelMaskParameter(None=@n1.Tech Art Mask) {group=Masks, maskChannel=Alpha, parameterName=GoopyMask, defaultValue=<rgba 0 0 0 1>}
  n2 = ChannelMaskParameter(None=@n1.Tech Art Mask) {group=Masks, parameterName=VeinMask}
  n3 = ChannelMaskParameter(None=@n1.Tech Art Mask) {group=Masks, maskChannel=Green, parameterName=LeatherySkinMask, defaultValue=<rgba 0 1 0 0>}
  n4 = ChannelMaskParameter(None=@n1.Tech Art Mask) {group=Masks, maskChannel=Blue, parameterName=SlimySackMaskColor, defaultValue=<rgba 0 0 1 0>}
  n5 = MaterialFunctionCall!(
   Mask=@n2.R
   Mat=MaterialFunctionCall!(
    Alpha=Max(A=@n3.R, B=@n2.R)
    M=@n1.M
    Material=MaterialFunctionCall!(Mask=@n4.R, M=@n1.M, Localpos=@n1.M, Mat=@n1.M) {materialFunction=&/MF_PistolCanary_SlimySack}.Result
   ) {materialFunction=&/MF_PistolCanary_LeatherySkin}.Result
  ) {materialFunction=&/MF_PistolCanary_PulsingVeins}
  MaterialFunctionCall(
   Material=StaticSwitch(
    True=MaterialFunctionCall!(Alpha=@n1.Tech Art Mask Alpha, Material=@n5.Result) {materialFunction=&/MF_PistolCanary_GoopyInt}.Result
    False=MaterialFunctionCall!(In=@n5.Result) {materialFunction=&/MF_WorldGlow_PistolCanary}.Result
    Value=StaticBoolParameter() {parameterName=UseGoopuInt}
   )
   M=@n1.M
   Pre Skinned Local Position=@n1.M
   Distance Blend=@n1.M
   Fresnel=@n1.M
  ) {materialFunction=&MF_CharacterParent_2023_PostFX}
  Reroute(None=@n1.Result)
 ]
}
