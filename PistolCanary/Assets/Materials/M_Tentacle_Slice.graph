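# M_Tentacle_Slice is a Default Lit, material-attributes material for PistolCanary, apparently for a tentacle slice, built on the 2023 character parent functions.
# The PreFX function supplies the base and a Tech Art Mask that channel parameters split into slimy sack, vein and leathery skin masks.
# Slimy sack applies over the maximum of sack and vein masks, then leathery skin layers on by its mask before PostFX finishes.
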Material /PistolCanary/Assets/Materials/M_Tentacle_Slice
graph {
 properties {bAutomaticallySetUsageInEditor=false, bUseMaterialAttributes=true, usedShadingModels=Default Lit}
 nodes [
  n1 = MaterialFunctionCall() {materialFunction=&MF_CharacterParent_2023_PreFX}
  n2 = Max(
   A=ChannelMaskParameter(None=@n1.Tech Art Mask) {group=Masks, maskChannel=Blue, parameterName=SlimySackMaskColor, defaultValue=<rgba 0 0 1 0>}.R
   B=ChannelMaskParameter(None=@n1.Tech Art Mask) {group=Masks, parameterName=VeinMask}.R
  )
  MaterialFunctionCall(
   Material=MaterialFunctionCall!(
    Alpha=ChannelMaskParameter(None=@n1.Tech Art Mask) {group=Masks, maskChannel=Green, parameterName=LeatherySkinMask, defaultValue=<rgba 0 1 0 0>}.R
    M=@n1.M
    Material=MaterialFunctionCall!(Mask=@n2, M=@n1.M, Localpos=@n1.M, Mat=@n1.M) {materialFunction=&/MF_PistolCanary_SlimySack}.Result
   ) {materialFunction=&/MF_PistolCanary_LeatherySkin}.Result
   M=@n1.M
   Pre Skinned Local Position=@n1.M
   Distance Blend=@n1.M
   Fresnel=@n1.M
  ) {materialFunction=&MF_CharacterParent_2023_PostFX}
 ]
}
