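# Physics asset for the tentacle creature skeletal mesh in PistolCanary, defining collision bodies and joints along one tentacle.
# Eighteen capsules, from tentacle_base through tentacle_17, taper toward the tip, with radii mostly far larger than their lengths.
# Constraints link each bone to the previous one in a single chain, with uniform swing and twist limit angles.
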PhysicsAsset /PistolCanary/Assets/Meshes/Tentacle_PistolCanary_Creature_Physics
bodies {
 tentacle_01 [{shapeType=Capsule, length=400, radius=1352.0885, rotation=<rot 0 0 90>}]
 tentacle_02 [{shapeType=Capsule, length=400, radius=1337.6562, rotation=<rot 0 0 90>}]
 tentacle_03 [{shapeType=Capsule, length=380, radius=1296.3506, rotation=<rot 0 0 90>}]
 tentacle_04 [{shapeType=Capsule, length=360, radius=1226.8732, rotation=<rot 0 0 90>}]
 tentacle_05 [{shapeType=Capsule, length=360, radius=1150.6189, rotation=<rot 0 0 90>}]
 tentacle_06 [{shapeType=Capsule, length=360, radius=1074.5582, rotation=<rot 0 0 90>}]
 tentacle_07 [{shapeType=Capsule, length=360, radius=970.3103, rotation=<rot 0 0 90>}]
 tentacle_08 [{shapeType=Capsule, length=360, radius=895.9664, rotation=<rot 0 0 90>}]
 tentacle_09 [{shapeType=Capsule, length=360, radius=809.4272, rotation=<rot 0 0 90>}]
 tentacle_10 [{shapeType=Capsule, length=360, radius=699.9668, rotation=<rot 0 0 90>}]
 tentacle_11 [{shapeType=Capsule, length=340, radius=630.2486, rotation=<rot 0 0 90>}]
 tentacle_12 [{shapeType=Capsule, length=340, radius=569.7941, rotation=<rot 0 0 90>}]
 tentacle_13 [{shapeType=Capsule, length=340, radius=498.331, rotation=<rot 0 0 90>}]
 tentacle_14 [{shapeType=Capsule, length=320, radius=440.7008, rotation=<rot 0 0 90>}]
 tentacle_15 [{shapeType=Capsule, length=320, radius=361.7017, rotation=<rot 0 0 90>}]
 tentacle_16 [{shapeType=Capsule, length=300, radius=288.8427, rotation=<rot 0 0 90>}]
 tentacle_17 [{shapeType=Capsule, length=300, radius=201.1644, rotation=<rot 0 0 90>}]
 tentacle_base [{shapeType=Capsule, length=400, radius=1430.0071, rotation=<rot 0 0 90>}]
}
constraints [
 {bone1Name=tentacle_01, bone2Name=tentacle_base, swing1LimitDegrees=45, swing2LimitDegrees=45, twistLimitDegrees=45}
 {bone1Name=tentacle_02, bone2Name=tentacle_01, swing1LimitDegrees=45, swing2LimitDegrees=45, twistLimitDegrees=45}
 {bone1Name=tentacle_03, bone2Name=tentacle_02, swing1LimitDegrees=45, swing2LimitDegrees=45, twistLimitDegrees=45}
 {bone1Name=tentacle_04, bone2Name=tentacle_03, swing1LimitDegrees=45, swing2LimitDegrees=45, twistLimitDegrees=45}
 {bone1Name=tentacle_05, bone2Name=tentacle_04, swing1LimitDegrees=45, swing2LimitDegrees=45, twistLimitDegrees=45}
 {bone1Name=tentacle_06, bone2Name=tentacle_05, swing1LimitDegrees=45, swing2LimitDegrees=45, twistLimitDegrees=45}
 {bone1Name=tentacle_07, bone2Name=tentacle_06, swing1LimitDegrees=45, swing2LimitDegrees=45, twistLimitDegrees=45}
 {bone1Name=tentacle_08, bone2Name=tentacle_07, swing1LimitDegrees=45, swing2LimitDegrees=45, twistLimitDegrees=45}
 {bone1Name=tentacle_09, bone2Name=tentacle_08, swing1LimitDegrees=45, swing2LimitDegrees=45, twistLimitDegrees=45}
 {bone1Name=tentacle_10, bone2Name=tentacle_09, swing1LimitDegrees=45, swing2LimitDegrees=45, twistLimitDegrees=45}
 {bone1Name=tentacle_11, bone2Name=tentacle_10, swing1LimitDegrees=45, swing2LimitDegrees=45, twistLimitDegrees=45}
 {bone1Name=tentacle_12, bone2Name=tentacle_11, swing1LimitDegrees=45, swing2LimitDegrees=45, twistLimitDegrees=45}
 {bone1Name=tentacle_13, bone2Name=tentacle_12, swing1LimitDegrees=45, swing2LimitDegrees=45, twistLimitDegrees=45}
 {bone1Name=tentacle_14, bone2Name=tentacle_13, swing1LimitDegrees=45, swing2LimitDegrees=45, twistLimitDegrees=45}
 {bone1Name=tentacle_15, bone2Name=tentacle_14, swing1LimitDegrees=45, swing2LimitDegrees=45, twistLimitDegrees=45}
 {bone1Name=tentacle_16, bone2Name=tentacle_15, swing1LimitDegrees=45, swing2LimitDegrees=45, twistLimitDegrees=45}
 {bone1Name=tentacle_17, bone2Name=tentacle_16, swing1LimitDegrees=45, swing2LimitDegrees=45, twistLimitDegrees=45}
]
