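# Physics asset for the PistolCanary creature's tentacle mesh, giving the tentacle and its five eyes capsule bodies for collision and physics.
# A chain of tapering capsules runs from tentacle_base to tentacle_17, with eye assemblies A–E branching off along it.
# In each eye, a mover body carries the orb and outer lids, and the orb carries the ball and inner lids.
# All constraints share uniform swing and twist limit values, without per-joint tuning.
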PhysicsAsset /PistolCanary/Assets/Meshes/Tentacle_PistolCanary_Creature_PhysicsAsset
bodies {
 eyeA_ball [{shapeType=Capsule, length=260.2673, radius=205.861, center=<773.7282 21.7803 2.4055>, rotation=<rot -3.8756 -90 -0.3813>}]
 eyeA_innerlidA [{shapeType=Capsule, length=406.7355, radius=157.3108, center=<-108.2225 7.9153 131.8722>, rotation=<rot 0 179.2794 -89.282>}]
 eyeA_lid_lowerA [{shapeType=Capsule, length=264.0662, radius=95.5513, center=<15.3322 -40.74 15.6694>, rotation=<rot -1.7653 -90 -0.8465>}]
 eyeA_lid_lowerB [{shapeType=Capsule, length=265.9862, radius=84.9985, center=<7.3467 2.6739 -43.534>, rotation=<rot 4.1876 -90 2.7892>}]
 eyeA_lid_upperA [{shapeType=Capsule, length=302.994, radius=158.528, center=<-48.606 26.9604 117.7331>, rotation=<rot 22.8732 -90 -6.3098>}]
 eyeA_lid_upperB [{shapeType=Capsule, length=314.3687, radius=187.0732, center=<-46.7077 68.6762 -93.4129>, rotation=<rot -23.9189 -90 5.3624>}]
 eyeA_mover [{shapeType=Capsule, length=1000.0167, radius=667.5336, center=<412.2558 240.5011 -3.656>, rotation=<rot -2.1143 -90 -1.415>}]
 eyeA_orb [{shapeType=Capsule, length=611.1681, radius=423.0931, center=<523.6529 12.0553 16.9243>, rotation=<rot -3.5624 -90 -0.1652>}]
 eyeB_ball [{shapeType=Capsule, length=206.1708, radius=110.2708, center=<761.4758 29.7926 15.6528>, rotation=<rot 1.7577 -90 -1.9868>}]
 eyeB_innerlidA [{shapeType=Capsule, length=413.2097, radius=145.1774, center=<-125.9963 -13.8636 116.5883>, rotation=<rot 0 7.1953 -89.5832>}]
 eyeB_innerlidB [{shapeType=Capsule, length=52.144, radius=32.8554, center=<15.0343 46.4657 -18.6227>, rotation=<rot 0 -171.7722 -61.5216>}]
 eyeB_lid_lowerA [{shapeType=Capsule, length=271.0667, radius=65.8134, center=<4.2367 -5.6874 60.3079>, rotation=<rot -5.9875 -90 1.918>}]
 eyeB_lid_lowerB [{shapeType=Capsule, length=163.8036, radius=53.8432, center=<-15.2489 6.195 -122.8492>, rotation=<rot 3.2015 -90 3.417>}]
 eyeB_lid_upperA [{shapeType=Capsule, length=285.9664, radius=118.4976, center=<52.2175 51.2896 105.9398>, rotation=<rot 22.2863 -90 9.1085>}]
 eyeB_lid_upperB [{shapeType=Capsule, length=300.9031, radius=151.8656, center=<50.8293 83.7681 -95.5608>, rotation=<rot -21.9398 -90 -10.0942>}]
 eyeB_mover [{shapeType=Capsule, length=1000.2736, radius=641.3653, center=<458.7314 -224.0265 104.5094>, rotation=<rot 4.2618 -90 2.2278>}]
 eyeB_orb [{shapeType=Capsule, length=613.968, radius=421.1621, center=<586.6894 -25.4177 54.2977>, rotation=<rot 2.113 -90 -0.8227>}]
 eyeC_ball [{shapeType=Capsule, length=331.2235, radius=187.5688, center=<733.1869 -10.8611 1.7692>, rotation=<rot -2.5247 -90 -0.209>}]
 eyeC_innerlidA [{shapeType=Capsule, length=246.112, radius=137.1513, center=<-67.9472 -29.1689 128.7032>, rotation=<rot 0 -2.5547 -88.3123>}]
 eyeC_lid_lowerA [{shapeType=Capsule, length=166.6145, radius=91.6695, center=<-32.9792 -5.7044 86.8332>, rotation=<rot 7.3434 -90 -4.328>}]
 eyeC_lid_lowerB [{shapeType=Capsule, length=108.2576, radius=67.3185, center=<-8.3847 4.304 -142.3773>, rotation=<rot -41.9304 -90 10.8018>}]
 eyeC_lid_upperA [{shapeType=Capsule, length=124.8248, radius=97.0064, center=<49.6762 62.4574 89.6959>, rotation=<rot 0 -31.6401 -86.0781>}]
 eyeC_lid_upperB [{shapeType=Capsule, length=178.7103, radius=127.3336, center=<42.5931 54.4013 -155.6113>, rotation=<rot -23.7669 -90 -7.1002>}]
 eyeC_mover [{shapeType=Capsule, length=1017.8026, radius=461.1018, center=<605.3327 78.7933 -39.0048>, rotation=<rot -2.3819 -90 -1.0281>}]
 eyeC_orb [{shapeType=Capsule, length=502.7832, radius=273.7841, center=<618.4282 -19.5721 11.0855>, rotation=<rot -1.8988 -90 -0.4473>}]
 eyeD_ball [{shapeType=Capsule, length=295.9023, radius=183.2716, center=<696.7297 -2.0962 6.6719>, rotation=<rot 2.3406 -90 -4.3516>}]
 eyeD_innerlidA [{shapeType=Capsule, length=324.3157, radius=149.049, center=<-132.0085 13.7486 129.6034>, rotation=<rot 0 -8.0546 -85.2533>}]
 eyeD_innerlidB [{shapeType=Capsule, length=116.6986, radius=32.3835, center=<-20.9014 -15.9835 -7.287>, rotation=<rot 0 -12.056 -63.2007>}]
 eyeD_lid_lowerA [{shapeType=Capsule, length=193.8188, radius=54.5188, center=<32.7543 -16.0145 13.7484>, rotation=<rot -5.9056 -90 8.2087>}]
 eyeD_lid_lowerB [{shapeType=Capsule, length=160.0084, radius=60.2262, center=<26.2832 7.8913 -1.0748>, rotation=<rot 4.04 -90 0.6003>}]
 eyeD_lid_upperA [{shapeType=Capsule, length=173.451, radius=136.8304, center=<-40.3309 59.9223 42.0529>, rotation=<rot 23.1365 -90 -4.316>}]
 eyeD_lid_upperB [{shapeType=Capsule, length=245.8841, radius=127.9043, center=<-30.4192 74.8072 -155.4418>, rotation=<rot -19.4261 -90 6.3374>}]
 eyeD_mover [{shapeType=Capsule, length=901.1723, radius=609.8079, center=<533.355 158.3935 -65.4129>, rotation=<rot 14.6985 -90 -5.7509>}]
 eyeD_orb [{shapeType=Capsule, length=506.7451, radius=297.4676, center=<590.1444 18.5307 14.0586>, rotation=<rot 1.214 -90 -2.9805>}]
 eyeE_ball [{shapeType=Capsule, length=138.2202, radius=63.7524, center=<492.145 -1.4744 -12.7996>, rotation=<rot -0.8924 -90 -3.3303>}]
 eyeE_innerlidA [{shapeType=Capsule, length=218.7929, radius=99.0522, center=<-78.0722 -3.3733 70.7068>, rotation=<rot 0 -177.8748 -89.2821>}]
 eyeE_innerlidB [{shapeType=Capsule, length=117.4635, radius=33.6693, center=<-10.9682 -7.5425 -22.6049>, rotation=<rot 0 178.6274 -87.6061>}]
 eyeE_lid_lowerB [{shapeType=Capsule, length=214.7541, radius=139.1187, center=<30.6402 18.243 -152.0602>, rotation=<rot 13.1043 -90 6.8864>}]
 eyeE_lid_upperA [{shapeType=Capsule, length=132.0739, radius=57.0851, center=<-30.4739 14.5814 72.5089>, rotation=<rot 0 -169.0154 -58.556>}]
 eyeE_lid_upperB [{shapeType=Capsule, length=192.1219, radius=142.5085, center=<-28.5314 22.3419 -119.9911>, rotation=<rot -22.6532 -90 2.3839>}]
 eyeE_mover [{shapeType=Capsule, length=811.9597, radius=440.0292, center=<616.3553 24.9784 -68.4061>, rotation=<rot 5.4091 -90 -4.3312>}]
 eyeE_orb [{shapeType=Capsule, length=381.1096, radius=199.7632, center=<408.8243 -9.3527 8.459>, rotation=<rot -0.8131 -90 -4.0124>}]
 tentacle_01 [{shapeType=Capsule, length=1472.3756, radius=1352.0885, center=<594.2203 -2.2731 -18.3359>, rotation=<rot 0 176.1069 -83.2777>}]
 tentacle_02 [{shapeType=Capsule, length=1451.6237, radius=1337.6562, center=<631.171 46.3696 6.5927>, rotation=<rot 0 -173.7945 -87.3944>}]
 tentacle_03 [{shapeType=Capsule, length=1345.4365, radius=1296.3506, center=<571.4348 -21.9125 -17.8614>, rotation=<rot 0 -4.007 -88.7887>}]
 tentacle_04 [{shapeType=Capsule, length=1319.1227, radius=1226.8732, center=<503.1753 -40.0073 -30.7523>, rotation=<rot 0 -179.526 -85.185>}]
 tentacle_05 [{shapeType=Capsule, length=1311.9987, radius=1150.6189, center=<569.7352 -49.182 -36.0448>, rotation=<rot 0 -173.5898 -79.667>}]
 tentacle_06 [{shapeType=Capsule, length=1275.3043, radius=1074.5582, center=<692.8198 -68.6571 -0.1925>, rotation=<rot 0 174.9176 -88.5947>}]
 tentacle_07 [{shapeType=Capsule, length=1170.908, radius=970.3103, center=<524.4554 -34.1811 -19.8427>, rotation=<rot 0 175.8796 -88.2091>}]
 tentacle_08 [{shapeType=Capsule, length=1054.411, radius=895.9664, center=<547.4081 89.3505 -24.2636>, rotation=<rot 0 3.2915 -79.2871>}]
 tentacle_09 [{shapeType=Capsule, length=960.861, radius=809.4272, center=<546.8607 44.6756 -21.6274>, rotation=<rot 0 0.0514 -84.3107>}]
 tentacle_10 [{shapeType=Capsule, length=864.246, radius=699.9668, center=<456.9672 10.527 -21.1715>, rotation=<rot 0 -0.6843 -86.6179>}]
 tentacle_11 [{shapeType=Capsule, length=776.8842, radius=630.2486, center=<353.4253 0.8005 -0.1704>, rotation=<rot 0 -179.9724 -89.9322>}]
 tentacle_12 [{shapeType=Capsule, length=704.8176, radius=569.7941, center=<317.8904 2.714 -6.5221>, rotation=<rot 0 -179.989 -89.9747>}]
 tentacle_13 [{shapeType=Capsule, length=618.6477, radius=498.331, center=<294.4701 1.2418 -3.5127>, rotation=<rot 0 179.9831 -89.9933>}]
 tentacle_14 [{shapeType=Capsule, length=544.004, radius=440.7008, center=<349.2005 1.2766 -3.3469>, rotation=<rot 0 -0.0332 -89.9593>}]
 tentacle_15 [{shapeType=Capsule, length=456.2953, radius=361.7017, center=<248.1599 -0.3748 -11.0439>, rotation=<rot 0 179.7905 -89.9799>}]
 tentacle_16 [{shapeType=Capsule, length=370.6425, radius=288.8427, center=<206.8067 -0.7948 -6.5972>, rotation=<rot 0 179.7566 -89.9028>}]
 tentacle_17 [{shapeType=Capsule, length=302.5728, radius=201.1644, center=<146.7114 -0.9875 -9.0164>, rotation=<rot 0 179.931 -89.9302>}]
 tentacle_base [{shapeType=Capsule, length=1439.8053, radius=1430.0071, center=<672.2713 0.3329 -102.6506>, rotation=<rot 0 179.9182 -88.0424>}]
}
constraints [
 {bone1Name=tentacle_01, bone2Name=tentacle_base, swing1LimitDegrees=45, swing2LimitDegrees=45, twistLimitDegrees=45}
 {bone1Name=tentacle_02, bone2Name=tentacle_01, swing1LimitDegrees=45, swing2LimitDegrees=45, twistLimitDegrees=45}
 {bone1Name=tentacle_03, bone2Name=tentacle_02, swing1LimitDegrees=45, swing2LimitDegrees=45, twistLimitDegrees=45}
 {bone1Name=tentacle_04, bone2Name=tentacle_03, swing1LimitDegrees=45, swing2LimitDegrees=45, twistLimitDegrees=45}
 {bone1Name=tentacle_05, bone2Name=tentacle_04, swing1LimitDegrees=45, swing2LimitDegrees=45, twistLimitDegrees=45}
 {bone1Name=tentacle_06, bone2Name=tentacle_05, swing1LimitDegrees=45, swing2LimitDegrees=45, twistLimitDegrees=45}
 {bone1Name=tentacle_07, bone2Name=tentacle_06, swing1LimitDegrees=45, swing2LimitDegrees=45, twistLimitDegrees=45}
 {bone1Name=tentacle_08, bone2Name=tentacle_07, swing1LimitDegrees=45, swing2LimitDegrees=45, twistLimitDegrees=45}
 {bone1Name=tentacle_09, bone2Name=tentacle_08, swing1LimitDegrees=45, swing2LimitDegrees=45, twistLimitDegrees=45}
 {bone1Name=tentacle_10, bone2Name=tentacle_09, swing1LimitDegrees=45, swing2LimitDegrees=45, twistLimitDegrees=45}
 {bone1Name=tentacle_11, bone2Name=tentacle_10, swing1LimitDegrees=45, swing2LimitDegrees=45, twistLimitDegrees=45}
 {bone1Name=tentacle_12, bone2Name=tentacle_11, swing1LimitDegrees=45, swing2LimitDegrees=45, twistLimitDegrees=45}
 {bone1Name=tentacle_13, bone2Name=tentacle_12, swing1LimitDegrees=45, swing2LimitDegrees=45, twistLimitDegrees=45}
 {bone1Name=tentacle_14, bone2Name=tentacle_13, swing1LimitDegrees=45, swing2LimitDegrees=45, twistLimitDegrees=45}
 {bone1Name=tentacle_15, bone2Name=tentacle_14, swing1LimitDegrees=45, swing2LimitDegrees=45, twistLimitDegrees=45}
 {bone1Name=tentacle_16, bone2Name=tentacle_15, swing1LimitDegrees=45, swing2LimitDegrees=45, twistLimitDegrees=45}
 {bone1Name=tentacle_17, bone2Name=tentacle_16, swing1LimitDegrees=45, swing2LimitDegrees=45, twistLimitDegrees=45}
 {bone1Name=eyeE_mover, bone2Name=tentacle_10, swing1LimitDegrees=45, swing2LimitDegrees=45, twistLimitDegrees=45}
 {bone1Name=eyeE_orb, bone2Name=eyeE_mover, swing1LimitDegrees=45, swing2LimitDegrees=45, twistLimitDegrees=45}
 {bone1Name=eyeE_innerlidB, bone2Name=eyeE_orb, swing1LimitDegrees=45, swing2LimitDegrees=45, twistLimitDegrees=45}
 {bone1Name=eyeE_innerlidA, bone2Name=eyeE_orb, swing1LimitDegrees=45, swing2LimitDegrees=45, twistLimitDegrees=45}
 {bone1Name=eyeE_ball, bone2Name=eyeE_orb, swing1LimitDegrees=45, swing2LimitDegrees=45, twistLimitDegrees=45}
 {bone1Name=eyeE_lid_upperB, bone2Name=eyeE_mover, swing1LimitDegrees=45, swing2LimitDegrees=45, twistLimitDegrees=45}
 {bone1Name=eyeE_lid_upperA, bone2Name=eyeE_mover, swing1LimitDegrees=45, swing2LimitDegrees=45, twistLimitDegrees=45}
 {bone1Name=eyeE_lid_lowerB, bone2Name=eyeE_mover, swing1LimitDegrees=45, swing2LimitDegrees=45, twistLimitDegrees=45}
 {bone1Name=eyeD_mover, bone2Name=tentacle_08, swing1LimitDegrees=45, swing2LimitDegrees=45, twistLimitDegrees=45}
 {bone1Name=eyeD_orb, bone2Name=eyeD_mover, swing1LimitDegrees=45, swing2LimitDegrees=45, twistLimitDegrees=45}
 {bone1Name=eyeD_innerlidA, bone2Name=eyeD_orb, swing1LimitDegrees=45, swing2LimitDegrees=45, twistLimitDegrees=45}
 {bone1Name=eyeD_innerlidB, bone2Name=eyeD_orb, swing1LimitDegrees=45, swing2LimitDegrees=45, twistLimitDegrees=45}
 {bone1Name=eyeD_ball, bone2Name=eyeD_orb, swing1LimitDegrees=45, swing2LimitDegrees=45, twistLimitDegrees=45}
 {bone1Name=eyeD_lid_upperB, bone2Name=eyeD_mover, swing1LimitDegrees=45, swing2LimitDegrees=45, twistLimitDegrees=45}
 {bone1Name=eyeD_lid_upperA, bone2Name=eyeD_mover, swing1LimitDegrees=45, swing2LimitDegrees=45, twistLimitDegrees=45}
 {bone1Name=eyeD_lid_lowerB, bone2Name=eyeD_mover, swing1LimitDegrees=45, swing2LimitDegrees=45, twistLimitDegrees=45}
 {bone1Name=eyeD_lid_lowerA, bone2Name=eyeD_mover, swing1LimitDegrees=45, swing2LimitDegrees=45, twistLimitDegrees=45}
 {bone1Name=eyeC_mover, bone2Name=tentacle_06, swing1LimitDegrees=45, swing2LimitDegrees=45, twistLimitDegrees=45}
 {bone1Name=eyeC_orb, bone2Name=eyeC_mover, swing1LimitDegrees=45, swing2LimitDegrees=45, twistLimitDegrees=45}
 {bone1Name=eyeC_innerlidA, bone2Name=eyeC_orb, swing1LimitDegrees=45, swing2LimitDegrees=45, twistLimitDegrees=45}
 {bone1Name=eyeC_ball, bone2Name=eyeC_orb, swing1LimitDegrees=45, swing2LimitDegrees=45, twistLimitDegrees=45}
 {bone1Name=eyeC_lid_upperB, bone2Name=eyeC_mover, swing1LimitDegrees=45, swing2LimitDegrees=45, twistLimitDegrees=45}
 {bone1Name=eyeC_lid_upperA, bone2Name=eyeC_mover, swing1LimitDegrees=45, swing2LimitDegrees=45, twistLimitDegrees=45}
 {bone1Name=eyeC_lid_lowerB, bone2Name=eyeC_mover, swing1LimitDegrees=45, swing2LimitDegrees=45, twistLimitDegrees=45}
 {bone1Name=eyeC_lid_lowerA, bone2Name=eyeC_mover, swing1LimitDegrees=45, swing2LimitDegrees=45, twistLimitDegrees=45}
 {bone1Name=eyeB_mover, bone2Name=tentacle_05, swing1LimitDegrees=45, swing2LimitDegrees=45, twistLimitDegrees=45}
 {bone1Name=eyeB_orb, bone2Name=eyeB_mover, swing1LimitDegrees=45, swing2LimitDegrees=45, twistLimitDegrees=45}
 {bone1Name=eyeB_innerlidB, bone2Name=eyeB_orb, swing1LimitDegrees=45, swing2LimitDegrees=45, twistLimitDegrees=45}
 {bone1Name=eyeB_innerlidA, bone2Name=eyeB_orb, swing1LimitDegrees=45, swing2LimitDegrees=45, twistLimitDegrees=45}
 {bone1Name=eyeB_ball, bone2Name=eyeB_orb, swing1LimitDegrees=45, swing2LimitDegrees=45, twistLimitDegrees=45}
 {bone1Name=eyeB_lid_upperB, bone2Name=eyeB_mover, swing1LimitDegrees=45, swing2LimitDegrees=45, twistLimitDegrees=45}
 {bone1Name=eyeB_lid_upperA, bone2Name=eyeB_mover, swing1LimitDegrees=45, swing2LimitDegrees=45, twistLimitDegrees=45}
 {bone1Name=eyeB_lid_lowerB, bone2Name=eyeB_mover, swing1LimitDegrees=45, swing2LimitDegrees=45, twistLimitDegrees=45}
 {bone1Name=eyeB_lid_lowerA, bone2Name=eyeB_mover, swing1LimitDegrees=45, swing2LimitDegrees=45, twistLimitDegrees=45}
 {bone1Name=eyeA_mover, bone2Name=tentacle_03, swing1LimitDegrees=45, swing2LimitDegrees=45, twistLimitDegrees=45}
 {bone1Name=eyeA_orb, bone2Name=eyeA_mover, swing1LimitDegrees=45, swing2LimitDegrees=45, twistLimitDegrees=45}
 {bone1Name=eyeA_innerlidA, bone2Name=eyeA_orb, swing1LimitDegrees=45, swing2LimitDegrees=45, twistLimitDegrees=45}
 {bone1Name=eyeA_ball, bone2Name=eyeA_orb, swing1LimitDegrees=45, swing2LimitDegrees=45, twistLimitDegrees=45}
 {bone1Name=eyeA_lid_upperB, bone2Name=eyeA_mover, swing1LimitDegrees=45, swing2LimitDegrees=45, twistLimitDegrees=45}
 {bone1Name=eyeA_lid_upperA, bone2Name=eyeA_mover, swing1LimitDegrees=45, swing2LimitDegrees=45, twistLimitDegrees=45}
 {bone1Name=eyeA_lid_lowerB, bone2Name=eyeA_mover, swing1LimitDegrees=45, swing2LimitDegrees=45, twistLimitDegrees=45}
 {bone1Name=eyeA_lid_lowerA, bone2Name=eyeA_mover, swing1LimitDegrees=45, swing2LimitDegrees=45, twistLimitDegrees=45}
]
